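# Niagara trail effect for the Spectacle Web pickaxe, drawing ribbons between socket pairs on the user-supplied skeletal mesh.
# Four anim-trail emitters each spawn from a different pair, with a random material parameter from a fixed seed.
# Three shape width and alpha by normalized execution index, then apply strong curve-scaled downward gravity and vector noise, distorting the strands.
# The fourth is a fainter, shorter-lived trail without forces that fades out.
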NiagaraSystem /BRCosmetics/Weapons/FORT_Melee/Pickaxe_Spectacle_Web_Male/FX/NS_Pickaxe_Spectacle_Web_Trail
dependencies {usedDataInterfaces [&NiagaraDataInterfaceCurve], usedDynamicInputs [], usedModules [&SystemState]}
emitters {
 Wyeth_AnimTrails {
  emitterUpdateScript [
   &EmitterState(Life Cycle Mode=::NewEnumerator1, Loop Behavior=::NewEnumerator1, Loop Duration=0.25, UseLoopDelay=false)
   &self:SpawnAnimTrail(Mesh:NiagaraDataInterfaceSkeletalMesh=$User.SkeletalMesh, Socket A=1) {module=SpawnAnimTrail}
  ]
  particleSpawnScript [
   &self:InitializeAnimTrail(Anim Trail Width Mode:ENiagaraAnimTrailWidthMode=::NewEnumerator1, Lifetime=0.2) {module=InitializeAnimTrail}
   &DynamicMaterialParameters(
    Index 0 Param 1=1
    Index 0 Param 2=&RandomRangeFloat(Minimum=0, Maximum=1, Randomness Mode=::NewEnumerator1, Random Seed=1, Fixed Random Seed=true)
    Module.Param0WriteEnabled=true
    Module.Param1WriteEnabled=true
    Module.Param2WriteEnabled=true
    Module.Param3WriteEnabled=true
   )
  ]
  particleUpdateScript [
   &ParticleState() {module=UpdateAge}
   &self:UpdateAnimTrail(Ribbon Width Scale:NiagaraFloat=&FloatFromCurve(
     FloatCurve=DataInterface() {propertyValues {Curve {keys [<at 0 0.25 RCIM_Cubic>,<at 0.5 1 RCIM_Cubic>,<at 1 0.25 RCIM_Cubic>]}}}
     CurveIndex=&ReturnNormalizedExecIndex()
    )) {module=UpdateAnimTrail}
   &ScaleColor(
    Scale Alpha=&FloatFromCurve(
     FloatCurve=DataInterface() {propertyValues {Curve {keys [
         {interpMode=RCIM_Cubic, tangentMode=RCTM_User, time=0, value=0}
         {interpMode=RCIM_Cubic, tangentMode=RCTM_User, time=0.5, value=1}
         {tangentMode=RCTM_User, time=1, value=0}
        ]}}}
     CurveIndex=&ReturnNormalizedExecIndex()
    )
    ScaleA=true
    ScaleRGB=true
   )
   &GravityForce(Gravity=&Multiply_VectorByFloat(
     Vector=<0 0 -15000>
     Float=&FloatFromCurve(
      FloatCurve=DataInterface() {propertyValues {Curve {keys [{interpMode=RCIM_Cubic, tangentMode=RCTM_User, time=0, value=1}, {tangentMode=RCTM_User, time=1, value=0}]}}}
      CurveIndex=&ReturnNormalizedExecIndex()
     )
    ))
   &VectorNoiseForce(Force Amount=7500, Module.Use Random Noise Time Interval=false)
   &SolveForcesAndVelocity(Clamp Velocity=false, Limit Acceleration=false)
  ]
  rendererClasses [&NiagaraRibbonRendererProperties]
  renderers [{rendererClass=&NiagaraRibbonRendererProperties}]
 }
 Wyeth_AnimTrails001 {
  emitterUpdateScript [
   &EmitterState(Life Cycle Mode=::NewEnumerator1, Loop Behavior=::NewEnumerator1, Loop Duration=0.25, UseLoopDelay=false)
   &self:SpawnAnimTrail(Mesh:NiagaraDataInterfaceSkeletalMesh=$User.SkeletalMesh, Socket A=3, Socket B=2) {module=SpawnAnimTrail}
  ]
  particleSpawnScript [
   &self:InitializeAnimTrail(Anim Trail Width Mode:ENiagaraAnimTrailWidthMode=::NewEnumerator1, Lifetime=0.25) {module=InitializeAnimTrail}
   &DynamicMaterialParameters(
    Index 0 Param 1=1
    Index 0 Param 2=&RandomRangeFloat(Minimum=0, Maximum=1, Randomness Mode=::NewEnumerator1, Random Seed=1, Fixed Random Seed=true)
    Module.Param0WriteEnabled=true
    Module.Param1WriteEnabled=true
    Module.Param2WriteEnabled=true
    Module.Param3WriteEnabled=true
   )
  ]
  particleUpdateScript [
   &ParticleState() {module=UpdateAge}
   &self:UpdateAnimTrail(Ribbon Width Scale:NiagaraFloat=&FloatFromCurve(
     FloatCurve=DataInterface() {propertyValues {Curve {keys [<at 0 0.25 RCIM_Cubic>,<at 1 1 RCIM_Cubic>]}}}
     CurveIndex=&ReturnNormalizedExecIndex()
    )) {module=UpdateAnimTrail}
   &ScaleColor(
    Scale Alpha=&FloatFromCurve(
     FloatCurve=DataInterface() {propertyValues {Curve {keys [<at 0 0 RCIM_Cubic>, {arriveTangentWeight=0.797, interpMode=RCIM_Cubic, tangentMode=RCTM_User, time=1, value=1}]}}}
     CurveIndex=&ReturnNormalizedExecIndex()
    )
    ScaleA=true
    ScaleRGB=true
   )
   &GravityForce(Gravity=&Multiply_VectorByFloat(
     Vector=<0 0 -15000>
     Float=&FloatFromCurve(
      FloatCurve=DataInterface() {propertyValues {Curve {keys [
          {interpMode=RCIM_Cubic, tangentMode=RCTM_User, time=0, value=1}
          {arriveTangent=-2.5053, interpMode=RCIM_Cubic, leaveTangent=-2.5053, tangentMode=RCTM_User, time=1, value=0}
         ]}}}
      CurveIndex=&ReturnNormalizedExecIndex()
     )
    ))
   &VectorNoiseForce(Force Amount=15000, Module.Use Random Noise Time Interval=false)
   &SolveForcesAndVelocity(Clamp Velocity=false, Limit Acceleration=false)
  ]
  rendererClasses [&NiagaraRibbonRendererProperties]
  renderers [{rendererClass=&NiagaraRibbonRendererProperties}]
 }
 Wyeth_AnimTrails002 {
  emitterUpdateScript [
   &EmitterState(Life Cycle Mode=::NewEnumerator1, Loop Behavior=::NewEnumerator1, Loop Duration=0.25, UseLoopDelay=false)
   &self:SpawnAnimTrail(Mesh:NiagaraDataInterfaceSkeletalMesh=$User.SkeletalMesh, Socket A=5, Socket B=4) {module=SpawnAnimTrail}
  ]
  particleSpawnScript [
   &self:InitializeAnimTrail(Anim Trail Width Mode:ENiagaraAnimTrailWidthMode=::NewEnumerator1, Lifetime=0.25) {module=InitializeAnimTrail}
   &DynamicMaterialParameters(
    Index 0 Param 1=1
    Index 0 Param 2=&RandomRangeFloat(Minimum=0, Maximum=1, Randomness Mode=::NewEnumerator1, Random Seed=1, Fixed Random Seed=true)
    Module.Param0WriteEnabled=true
    Module.Param1WriteEnabled=true
    Module.Param2WriteEnabled=true
    Module.Param3WriteEnabled=true
   )
  ]
  particleUpdateScript [
   &ParticleState() {module=UpdateAge}
   &self:UpdateAnimTrail(Ribbon Width Scale:NiagaraFloat=&FloatFromCurve(
     FloatCurve=DataInterface() {propertyValues {Curve {keys [<at 0 0.25 RCIM_Cubic>,<at 1 1 RCIM_Cubic>]}}}
     CurveIndex=&ReturnNormalizedExecIndex()
    )) {module=UpdateAnimTrail}
   &ScaleColor(
    Scale Alpha=&FloatFromCurve(
     FloatCurve=DataInterface() {propertyValues {Curve {keys [<at 0 0 RCIM_Cubic>, {arriveTangentWeight=0.797, interpMode=RCIM_Cubic, tangentMode=RCTM_User, time=1, value=1}]}}}
     CurveIndex=&ReturnNormalizedExecIndex()
    )
    ScaleA=true
    ScaleRGB=true
   )
   &GravityForce(Gravity=&Multiply_VectorByFloat(
     Vector=<0 0 -15000>
     Float=&FloatFromCurve(
      FloatCurve=DataInterface() {propertyValues {Curve {keys [
          {interpMode=RCIM_Cubic, tangentMode=RCTM_User, time=0, value=1}
          {arriveTangent=-2.5053, interpMode=RCIM_Cubic, leaveTangent=-2.5053, tangentMode=RCTM_User, time=1, value=0}
         ]}}}
      CurveIndex=&ReturnNormalizedExecIndex()
     )
    ))
   &VectorNoiseForce(Force Amount=15000, Module.Use Random Noise Time Interval=false)
   &SolveForcesAndVelocity(Clamp Velocity=false, Limit Acceleration=false)
  ]
  rendererClasses [&NiagaraRibbonRendererProperties]
  renderers [{rendererClass=&NiagaraRibbonRendererProperties}]
 }
 Wyeth_AnimTrails003 {
  emitterUpdateScript [
   &EmitterState(Life Cycle Mode=::NewEnumerator1, Loop Behavior=::NewEnumerator1, Loop Duration=0.25, UseLoopDelay=false)
   &self:SpawnAnimTrail(Mesh:NiagaraDataInterfaceSkeletalMesh=$User.SkeletalMesh, Socket A=6, Socket B=7) {module=SpawnAnimTrail}
  ]
  particleSpawnScript [
   &self:InitializeAnimTrail(Color:LinearColor=<rgba 1 1 1 0.25>, Lifetime=0.1, Ribbon Width Scale=1) {module=InitializeAnimTrail}
   &DynamicMaterialParameters(
    Index 0 Param 1=1
    Index 0 Param 2=&RandomRangeFloat(Minimum=0, Maximum=1, Randomness Mode=::NewEnumerator1, Random Seed=1, Fixed Random Seed=true)
    Module.Param0WriteEnabled=true
    Module.Param1WriteEnabled=true
    Module.Param2WriteEnabled=true
    Module.Param3WriteEnabled=true
   )
  ]
  particleUpdateScript [
   &ParticleState() {module=UpdateAge}
   &self:UpdateAnimTrail() {module=UpdateAnimTrail}
   &ScaleColor(
    Scale Alpha=&FloatFromCurve(FloatCurve=DataInterface() {propertyValues {Curve {keys [{arriveTangent=-2.2484, interpMode=RCIM_Cubic, leaveTangent=-2.2484, tangentMode=RCTM_User, time=0, value=1}, <at 1 0 RCIM_Cubic>]}}})
    ScaleA=true
    ScaleRGB=true
   )
   &SolveForcesAndVelocity(Clamp Velocity=false, Limit Acceleration=false)
  ]
  rendererClasses [&NiagaraRibbonRendererProperties]
  renderers [{rendererClass=&NiagaraRibbonRendererProperties}]
 }
}
summary {
 systemName=NS_Pickaxe_Spectacle_Web_Trail
 emitters [
  {emitterName=Wyeth_AnimTrails, rendererClasses [&NiagaraRibbonRendererProperties]}
  {emitterName=Wyeth_AnimTrails001, rendererClasses [&NiagaraRibbonRendererProperties]}
  {emitterName=Wyeth_AnimTrails002, rendererClasses [&NiagaraRibbonRendererProperties]}
  {emitterName=Wyeth_AnimTrails003, rendererClasses [&NiagaraRibbonRendererProperties]}
 ]
 userVariables [
  User.bisFrontEndPreview {type=NiagaraBool}
  User.bisFullyDeployed {type=NiagaraBool}
  User.ForwardDot {type=NiagaraFloat, defaultValue=0, description=""}
  User.RibbonIDOffset {type=NiagaraInt32}
  User.RibbonWidth {type=NiagaraFloat, defaultValue=10, description=""}
  User.SkeletalMesh {
   type=NiagaraDataInterfaceSkeletalMesh
   defaultValue {
    struct=NiagaraExt_VariableValue_DataInterface
    value {dataInterface=&self:NiagaraDataInterfaceSkeletalMesh_0}
   }
  }
 ]
}
