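# Physics asset for the male Spectacle Web pickaxe, the front-end variant, simulating the dangling web bones on the axe.
# Capsule bodies cover the two-handed handle, a web core and three web pieces.
# The three web pieces are each constrained to the core, which is constrained to the handle, forming a branched hierarchy.
# Narrow swing and twist limits restrict the motion, with the third web piece tighter in swing than the other two.
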PhysicsAsset /BRCosmetics/Weapons/FORT_Melee/Pickaxe_Spectacle_Web_Male/Meshes/Spectacle_Web_Male_Axe_Physics_FrontEnd
bodies {
 Melee_TwoHanded_Handle [{shapeType=Capsule, length=97.1745, radius=2.8943, shapeName=Melee_TwoHanded_Handle_capsule, center=<0.734 -1.069 21.628>}]
 dyn_web01 [{
   shapeType=Capsule
   length=7.1349
   radius=4.6772
   shapeName=dyn_web01_capsule
   center=<-2.865 -1.1598 -4.0173>
   rotation=<rot -38.077 0.0307 -5.1601>
  }]
 dyn_web02 [{
   shapeType=Capsule
   length=6.4094
   radius=4.2102
   shapeName=dyn_web02_capsule
   center=<-2.006 -0.2076 -3.5376>
   rotation=<rot -29.5557 0 -0.4837>
  }]
 dyn_web03 [{shapeType=Capsule, length=4.893, radius=3.7427, shapeName=dyn_web03_capsule, center=<0.385 1.5059 -3.232>, rotation=<rot -0.0003 0 -24.9821>}]
 dyn_webcore [{shapeType=Capsule, length=0, radius=3.3263, center=<-0.2138 0.3559 -2.1474>, rotation=<rot 18.9938 91.956 -4.2423>}]
}
constraints [
 {bone1Name=dyn_web01, bone2Name=dyn_webcore, swing1LimitDegrees=30, swing2LimitDegrees=30, twistLimitDegrees=5}
 {bone1Name=dyn_web02, bone2Name=dyn_webcore, swing1LimitDegrees=30, swing2LimitDegrees=30, twistLimitDegrees=5}
 {bone1Name=dyn_web03, bone2Name=dyn_webcore, swing1LimitDegrees=10, swing2LimitDegrees=10, twistLimitDegrees=5}
 {bone1Name=dyn_webcore, bone2Name=Melee_TwoHanded_Handle, swing1LimitDegrees=15, swing2LimitDegrees=15, twistLimitDegrees=10}
]
